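; BPI_LockerList_Type is a Blueprint Interface, apparently for the item list widget types in the Locker UI.
; It exposes getters for the entry object, list view, group list view and last focused column, plus setters for group counts and header row indices.
; Other functions bind preview-widget actions, set up telemetry, control context visibility and selection, and update focus-navigation and emote-wheel flags.
; It declares signatures only; implementers supply the behavior.
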(fn BindPreviewActions
  (params (Preview%20Widget :type Fort%20Item%20Preview%20Widget%20Object%20Reference)))

(fn GetEntryObject
  (return 0))

(fn GetGroupListView
  (return 0))

(fn GetLastFocusedColumn
  (params (LastFocusedColumnIndex :type Integer)))

(fn GetListView
  (return 0))

(fn SetContextVisibility)

(fn SetGroupCounts
  (params (ItemCountPerGroup :type Array%20of%20Integers)))

(fn SetHeaderIndex
  (params (FirstRowGroupArray :type Array%20of%20Integers)))

(fn SetupBinds
  (params (Widget :type Widget%20Object%20Reference) (PreviewWidget :type Fort%20Item%20Preview%20Widget%20Object%20Reference)))

(fn SetupPreviewActionBinds
  (params (PreviewWidget :type Fort%20Item%20Preview%20Widget%20Object%20Reference)))

(fn SetupTelemetry
  (params (TabFlowRiderString :type String)))

(fn ToggleSelectedContext)

(fn UpdateAllowNavOnFocus
  (params (AllowNavOnFocus :type Boolean)))

(fn UpdateIsEmoteWheel
  (params (IsEmoteWheel :type Boolean)))
