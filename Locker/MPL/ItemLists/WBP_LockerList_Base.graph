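# WBP_LockerList_Base is a base CommonUserWidget for locker item lists, with no widget tree of its own.
# It declares event dispatchers for selection closing, style editing, item details, modal changes, tile context menus and streaming waits.
# Stub functions for list view, group list view, entry object, preview actions and impression telemetry serve as a shared interface.
# OnMouseButtonDown broadcasts OnEntryRightClicked on right-click without handling the input.
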WidgetBlueprint /Locker/MPL/ItemLists/WBP_LockerList_Base
graphs [
 AddImpressionTelemetry {nodes [AddImpressionTelemetry!() -> (then, TabFlowRiderString:String) {is=title K2Node_FunctionEntry}]}
 CloseItemSelection {nodes [CloseItemSelection() {is=title K2Node_FunctionEntry}]}
 ContextEditStylesClicked {nodes [ContextEditStylesClicked() {is=title K2Node_FunctionEntry}]}
 EventGraph {nodes [
   n1 = Event PreConstruct()
   > Parent: PreConstruct(IsDesignTime=@n1.IsDesignTime)
   Event Construct()
   > Parent: Construct()
   n2 = Event Tick()
   > Parent: Tick!(execute, MyGeometry:Geometry Structure=@n2.MyGeometry, InDeltaTime:float=@n2.InDeltaTime) -> (then) {is=title K2Node_CallParentFunction}
  ]}
 GetEntryObject {nodes [
   GetEntryObject() {is=title K2Node_FunctionEntry}
   > Return Node!(execute, Object:Object Reference)
  ]}
 GetGroupListView {nodes [
   GetGroupListView() {is=title K2Node_FunctionEntry}
   > Return Node!(execute, ListView:Common List View Object Reference)
  ]}
 GetListView {nodes [
   GetListView() {is=title K2Node_FunctionEntry}
   > Return Node!(execute, TileView:Fort Tile View Object Reference)
  ]}
 GoToLockerItemDetails {nodes [GoToLockerItemDetails!() -> (then, CustomItemDetails:User Widget Class Reference) {is=title K2Node_FunctionEntry}]}
 OnEntryRightClicked {nodes [OnEntryRightClicked() {is=title K2Node_FunctionEntry}]}
 OnModalChanged {nodes [OnModalChanged!() -> (then, IsModalOpen:bool=~) {is=title K2Node_FunctionEntry}]}
 OnMouseButtonDown {nodes [
   n1 = OnMouseButtonDown!() -> (then, MyGeometry:Geometry Structure, MouseEvent:"Pointer Event Structure (by ref)") {is=title K2Node_FunctionEntry}
   > n2 = Branch(Condition="Equal (Key)"(
     A:Key Structure=Get Effecting Button!(Input:"Pointer Event Structure (by ref)"=@n1.MouseEvent) -> (ReturnValue:Key Structure) {is=title K2Node_CallFunction}
     B:Key Structure=RightMouseButton
    ) -> (ReturnValue:bool=~) {is=title K2Node_PromotableOperator})
   > Call On Entry Right Clicked!(execute:Exec, self:WBP Locker List Base Object Reference) -> (then:Exec) {is=title K2Node_CallDelegate}
   > Return Node!(execute, ReturnValue:Event Reply Structure=Unhandled())
   Return Node(execute=@n2.else, ReturnValue=Make Event Reply!() -> (EventReply:Event Reply Structure) {is=title K2Node_MakeStruct})
  ]}
 OnTileContextMenuClosed {nodes [OnTileContextMenuClosed() {is=title K2Node_FunctionEntry}]}
 OnTileContextMenuOpen {nodes [OnTileContextMenuOpen!() -> (then, Widget:Widget Object Reference) {is=title K2Node_FunctionEntry}]}
 OnWaitForStreamingContent {nodes [OnWaitForStreamingContent() {is=title K2Node_FunctionEntry}]}
 PreviewActions {nodes [PreviewActions!() -> (then, PreviewWidget:Fort Item Preview Widget Object Reference) {is=title K2Node_FunctionEntry}]}
]
variables [
 CloseItemSelection {}
 ContextEditStylesClicked {}
 GoToLockerItemDetails {}
 OnEntryRightClicked {}
 OnModalChanged {}
 OnTileContextMenuClosed {}
 OnTileContextMenuOpen {}
 OnWaitForStreamingContent {}
]
widgets {info {inheritedWidgetCount=0, parentClass=&CommonUserWidget, rootWidgetClass=null, widgetCount=0}, widgets []}
