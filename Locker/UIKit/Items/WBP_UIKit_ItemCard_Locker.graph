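# Locker item card widget for a tile view, bound to a FortItemVM, that handles selecting, equipping and customizing cosmetics.
# Item state and category config build context menu actions such as favorite, archive, apply to all, edit styles, rename and quests.
# Hover, focus, touch and right-click logic adapts menu visibility per input method.
# Equip and confirmation run as async actions, custom detail widgets load asynchronously, and dispatchers report selection, modal and streaming-wait events.
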WidgetBlueprint /Locker/UIKit/Items/WBP_UIKit_ItemCard_Locker
animations [
 Anim_Hovered {
  binding_count=1
  length_seconds=0.2004
  object=&self:Anim_Hovered
  sequence {bindings [WBP_UIKit_ItemCard_Locker_C_17 {
     id=#1
     possessed_class=WBP_UIKit_ItemCard_Locker_C
     tracks [{
       type=MovieScene2DTransformTrack
       sections [{
         type=MovieScene2DTransformSection
         end_frame=4
         start_frame=0
         channels [
          Translation.X {type=MovieSceneScriptingFloatChannel, const [0,1]}
          Translation.Y {type=MovieSceneScriptingFloatChannel, const [0,1]}
          Angle {type=MovieSceneScriptingFloatChannel, const [0,1]}
          Scale.X {type=MovieSceneScriptingFloatChannel, keys [[0,1],[4.007,1.03]]}
          Scale.Y {type=MovieSceneScriptingFloatChannel, keys [[0,1],[4.007,1.03]]}
          Shear.X {type=MovieSceneScriptingFloatChannel, const [0,1]}
          Shear.Y {type=MovieSceneScriptingFloatChannel, const [0,1]}
         ]
        }]
      }]
    }]}
 }
 FadeOutRewardPathing {
  binding_count=1
  length_seconds=0.25
  object=&self:FadeOutRewardPathing
  sequence {bindings [RewardPathingGlow {
     id=#2
     possessed_class=Image
     tracks [
      {
       type=MovieSceneWidgetMaterialTrack
       sections [{
         type=MovieSceneParameterSection
         end_frame=5
         start_frame=0
         channels [Fadeout_Amount {type=MovieSceneScriptingFloatChannel, keys [[0,0],[3,1],[4,1],[5,0]]}]
        }]
      }
      {
       type=MovieSceneByteTrack
       display_name=Visibility
       sections [{type=MovieSceneByteSection, end_frame=5, start_frame=0, channels [None {type=MovieSceneScriptingByteChannel, keys [[0,4],[3,4],[4,2]]}]}]
      }
     ]
    }]}
 }
]
graphs [
 AdjustContextMenuTouch {nodes [
   n1 = AdjustContextMenuTouch!() -> (then, IsTouchInputType:bool=~) {is=title K2Node_FunctionEntry}
   > Set Visibility(
    Visibility=Select(Option 0=Visible, Option 1=HitTestInvisible, Index=@n1)
    self=Get ContextMenuButton!(self:WBP UIKit Context Menu Anchor Button Base Object Reference=Get ContextMenu() -> (ContextMenu:WBP UIKit Context Menu Anchor Button Object Reference)) -> (ContextMenuButton:WBP UIKit Button Generic Base Object Reference)
   )
  ]}
 AnalyticsItemInteracted {nodes [
   n1 = AnalyticsItemInteracted!() -> (then, Context:String) {is=title K2Node_FunctionEntry}
   n2 = GetItemTemplateId!(self:Fort Item VM Object Reference=Get FortItemVM() -> (FortItemVM:Fort Item VM Object Reference)) -> (ReturnValue:String)
   > FireEventElementInteractedByName!(
    execute
    Element Name:String="To String (Text)"(InText=Format Text!(Format="LOCTEXT(\"{0} {1}\")", 0:String=@n1, 1:String=@n2) -> (Result))
    WidgetCallingFrom:Widget Object Reference=Self-Reference()
   ) -> (then, Success:bool)
  ]}
 BP_OnHoldTriggered {nodes [
   n1 = On Hold Triggered!() -> (then, CurrentInputType:ECommonInputType Enum) {is=title K2Node_FunctionEntry}
   > n2 = Switch on ECommonInputType!(execute, Selection:ECommonInputType Enum=@n1) -> (MouseAndKeyboard:Exec, Gamepad:Exec, Touch:Exec, Count:Exec) {is=title K2Node_SwitchEnum}
   ToggleContextMenuIfSelected(execute=@n2.Touch)
   > Return Node!(execute, ReturnValue:Event Reply Structure=Handled())
  ]}
 Can Display Context Menu {nodes [
   Can Display Context Menu() {is=title K2Node_FunctionEntry}
   n1 = Get FortItemVM()
   n2 = NOT Boolean(A=Get bIsValid!(self:Fort Item VM Object Reference=@n1) -> (bIsValid:bool))
   n3 = Get bIsShuffleTile!(self:Fort Item VM Object Reference=@n1) -> (bIsShuffleTile:bool)
   n4 = Get SelectedCategory!(self:Fort Item Selection VM Object Reference=Get FortItemSelectionVM() -> (FortItemSelectionVM:Fort Item Selection VM Object Reference)) -> (SelectedCategory:Fort Locker Category Item VM Object Reference)
   > Return Node!(
    execute
    Value:bool=AND Boolean(
     A=NOR Boolean!(A:bool=@n3, B:bool=@n2) -> (ReturnValue:bool)
     B=GetIsEquippable!(self:Fort Locker Category Item VM Object Reference=@n4) -> (ReturnValue:bool)
    )
   )
   n5 = GetItemType!(ItemDefinition:Item Definition Base Object Reference=Get ItemDefinition!(self:Fort Item VM Object Reference=@n1) -> (ItemDefinition:Fort Item Definition Object Reference)) -> (ReturnValue:EFortItemType Enum=WorldItem)
   NOR Boolean(A=AND Boolean(A="Equal (Enum)"(A:EFortItemType Enum=@n5, B:EFortItemType Enum=AthenaMusicPack), B=@n3), B=NOT Boolean(A=@n3))
  ]}
 CanReactToRightClickDown {nodes [
   n1 = CanReactToRightClickDown!() -> (then, Input:"Pointer Event Structure (by ref)") {is=title K2Node_FunctionEntry}
   n2 = Get ContextMenuActionContext() -> (ContextMenuActionContext:UIKit Context Menu Action Context Object Reference)
   n3 = IsOpen!(self:Menu Anchor Object Reference=Get Anchor!(self:UIKit Context Menu Anchor Object Reference=Get ContextMenu()) -> (Anchor:Menu Anchor Object Reference)) -> (ReturnValue:bool)
   > Return Node!(
    execute
    ReturnValue:bool=AND Boolean!(
     A=Is Mouse Button Down!(Input:"Pointer Event Structure (by ref)"=@n1, MouseButton:Key Structure=RightMouseButton) -> (ReturnValue:bool)
     B=NOT Boolean(A=@n3)
     C:bool=Can Display Context Menu!() -> (Value:bool) {is=title K2Node_CallFunction}
     D:bool=HasAvailableInteractions!(self:UIKit Context Menu Action Context Object Reference=@n2) -> (ReturnValue:bool)
    ) -> (ReturnValue)
   )
  ]}
 CloseContextMenu {nodes [
   CloseContextMenu() {is=title K2Node_FunctionEntry}
   > Close(self:Menu Anchor Object Reference=Get Anchor(self=Get ContextMenu()))
  ]}
 Convert_FortItemVMToSelectedConnectedQuestCategoryTag {nodes [
   n1 = Convert_FortItemVMToSelectedConnectedQuestCategoryTag!() -> (then, InItemVM:Fort Item VM Object Reference) {is=title K2Node_FunctionEntry}
   n2 = Get ItemDefinition(self=@n1)
   > Is Valid(InputObject=@n2)
   > Return Node!(
    execute
    OutGameplayTags:Gameplay Tag Structure=GetConnectedQuestCategoryTag!(ItemDefinition:Item Definition Base Object Reference=@n2) -> (ReturnValue:Gameplay Tag Structure)
   )
  ]}
 EventGraph {nodes [
   OnRightClickDown() {is=title K2Node_CustomEvent}
   > Open!(execute, self:Menu Anchor Object Reference=Get Anchor(self=Get ContextMenu()), bFocusMenu:bool=false) -> (then)
   > PlaySound2D(Sound=Get Right Clicked Sound() -> (Right Clicked Sound:Sound Base Object Reference))
   Event On Favorite Item!() -> (OutputDelegate, then, InteractionNameID:Name) {is=title K2Node_CustomEvent}
   n1 = Get Category!(self:Fort Locker Category Item VM Object Reference=Get SelectedCategory(self=Get FortItemSelectionVM())) -> (Category:Fort Locker Category VM Object Reference)
   n2 = Get FortItemVM()
   > ToggleItemFavorite!(execute, self:Fort Locker Category VM Object Reference=@n1, Item:Fort Item VM Object Reference=@n2) -> (then)
   > n3 = PlaySound2D(Sound=Select(
     Option 0:Sound Base Object Reference=Get Unfavorite Sound() -> (Unfavorite Sound:Sound Base Object Reference)
     Option 1:Sound Base Object Reference=Get Favorite Sound() -> (Favorite Sound:Sound Base Object Reference)
     Index=Get bIsFavorite!(self:Fort Item VM Object Reference=@n2) -> (bIsFavorite:bool)
    ) -> (ReturnValue:Sound Base Object Reference))
   CloseContextMenu(execute=[@n18.then, @n5.then, @n7.then, @n3.then, @n11.else]) {is=title K2Node_CallFunction}
   > n4 = RemoveAllInteractions(self:UIKit Context Menu Action Context Object Reference=Get ContextMenuActionContext())
   SetContextMenuActionContext(execute=[@n61.else, @n75.then, @n28.then, @n4.then])
   Event On Archive Item!() -> (OutputDelegate, then, InteractionNameID:Name) {is=title K2Node_CustomEvent}
   > n5 = ToggleItemArchived!(execute, self:Fort Locker Category VM Object Reference=@n1, Item:Fort Item VM Object Reference=@n2) -> (then)
   Apply to All!() -> (OutputDelegate, then, InteractionNameID:Name) {is=title K2Node_CustomEvent}
   > n6 = SetFIVMItemAndSlot!(execute, self) -> (then, Selected Item:Fort Item VM Object Reference, Slot Index:int=0)
   > n7 = EquipItem!(
    execute:Exec
    Item:Fort Item VM Object Reference=@n6
    SlotIndex:int=@n6.Slot Index
    Variants:Array of Mcp Variant Channel Info Structures=Make Array!() -> (Array:Array of Mcp Variant Channel Info Structures)
    bApplyToAllSlots:bool=true
    bAllowMocking:bool=true
   ) -> (then:Exec, OnOperationCompleted:Exec, Success:bool=~) {is=title K2Node_AsyncAction}
   EditStyles!() -> (OutputDelegate, then, InteractionNameID:Name) {is=title K2Node_CustomEvent}
   > Call On Context Menu Edit Styles Clicked!(execute:Exec, self:WBP UIKit Item Card Locker Object Reference) -> (then:Exec) {is=title K2Node_CallDelegate}
   Rename Modal!() -> (OutputDelegate, then, InteractionNameID:Name) {is=title K2Node_CustomEvent}
   n8 = Get FortItemSelectionVM()
   > n9 = EquipItem(
    Item=Get SelectedItem!(self:Fort Item Selection VM Object Reference=@n8) -> (SelectedItem:Fort Item VM Object Reference)
    SlotIndex=Get SlotIndex!(self:Fort Locker Category Item VM Object Reference=Get SelectedCategory(self=@n8)) -> (SlotIndex:int=0)
    Variants=Make Array()
    bApplyToAllSlots=false
    bAllowMocking=true
   ) -> (Success=~)
   n10 = TryGetSoftCustomActivatableWidgetClass!(
    self:Fort Locker Category Config Object Reference=GetItemCategory!(self:Fort Item VM Object Reference=Get FortItemVM()) -> (ReturnValue:Fort Locker Category Config Object Reference)
    ViewRequest:Gameplay Tag Structure=(TagName="UI.ItemPreview.PluginBridge.Mimosa.RenameModal")
   ) -> (WidgetClass:Common Activatable Widget Soft Class Reference, ReturnValue:bool)
   n11 = Branch(execute=@n9.OnOperationCompleted, Condition=@n10.ReturnValue)
   > n12 = LoadClassAsset_Blocking(AssetClass=@n10)
   > n13 = Cast To CommonActivatableWidget Class!(execute:Exec, Class:Object Class Reference=@n12) -> (then:Exec, CastFailed:Exec, AsCommon Activatable Widget:Common Activatable Widget Class Reference) {is=title K2Node_ClassDynamicCast}
   > n14 = Create Widget!(execute:Exec, Class:User Widget Class Reference=@n13, OwningPlayer:Player Controller Object Reference=GetOwningPlayer()) -> (then:Exec, ReturnValue:Common Activatable Widget Object Reference)
   > n15 = Set RenameModal(RenameModal:Common Activatable Widget Object Reference=@n14) -> (Output_Get:Common Activatable Widget Object Reference)
   > n16 = TryPushWidgetAsModalPanel!(execute, PushWidget:User Widget Object Reference=@n15) -> (then)
   n17 = ModalDeactivated() {is=title K2Node_CustomEvent}
   Bind Event to On Widget Deactivated(execute=@n16.then, self:Common Activatable Widget Object Reference=@n15, Delegate=@n17) {is=title K2Node_AddDelegate}
   > n18 = Call Is Modal Open!(execute:Exec, self:WBP UIKit Item Card Locker Object Reference, IsOpen:bool=true) -> (then:Exec) {is=title K2Node_CallDelegate}
   ToggleContextMenuIfSelected() {is=title K2Node_CustomEvent}
   n19 = GetCurrentInputType(self=Get CommonInputSubsystem())
   n20 = Get SelectedItem(self=Get FortItemSelectionVM())
   n21 = Get FortItemVM()
   > Branch(Condition=Select(
     Option 0:bool="Equal (Object)"(A=@n21, B=@n20)
     Option 1:bool=Get IsItemCardLockerHovered() -> (IsItemCardLockerHovered:bool)
     Index="Equal (Enum)"(A:ECommonInputType Enum=@n19, B:ECommonInputType Enum=Gamepad)
    ) -> (ReturnValue:bool))
   > Branch(Condition=Can Display Context Menu())
   > SetVisibility(self=Get ContextMenu(), InVisibility=Visible)
   > ToggleContextMenu(self:WBP UIKit Context Menu Anchor Button Base Object Reference=Get ContextMenu())
   n22 = Event PreConstruct()
   > Parent: PreConstruct(IsDesignTime=@n22.IsDesignTime)
   > SetGridAlignment!(
    execute
    self:WBP UIKit Item Card Base Object Reference
    WIdget:Widget Object Reference=Get Slot_Main() -> (Slot_Main:Named Slot Object Reference)
    Horizontal:EHorizontalAlignment Enum=HAlign_Right
    Vertical:EVerticalAlignment Enum=VAlign_Bottom
   ) -> (then)
   n23 = Get ContextMenu()
   > SetPadding!(
    execute
    self:Grid Slot Object Reference=SlotAsGridSlot(Widget=Get Anchor(self=@n23))
    InPadding_Left:float=0
    InPadding_Top:float=0
    InPadding_Right:float=0
    InPadding_Bottom:float=0
   ) -> (then)
   > SetVisibility(self=Get ContextMenu(), InVisibility=Hidden)
   n24 = Event OnListItemObjectSet()
   > n25 = Cast To FortItemVM!(execute, Object=@n24.ListItemObject) -> (then, CastFailed, AsFort Item VM:Fort Item VM Object Reference) {is=title K2Node_DynamicCast}
   > InitializeItemCard(self:WBP UIKit Item Card Base Object Reference)
   > SetFortItemVM!(execute, self, Viewmodel:Fort Item VM Object Reference=@n25) -> (then)
   > SetCachedIndicatorConfig!(
    execute
    self
    CachedIndicatorTypeToFind:E_UIKit_ItemCardStatusIndicator Enum=NewEnumerator5
    NewIndicatorConfig:E_UI_StatusIndicator_Configuration Enum=NewEnumerator2
   ) -> (then, Success:bool)
   > n26 = Is Valid(InputObject=Get ContextMenuActionContext())
   n27 = Construct Object from Class(execute=@n26.Is Not Valid, Class=/Script/FortniteUI.UIKitContextMenuActionContext) -> (ReturnValue:UIKit Context Menu Action Context Object Reference)
   > Set ContextMenuActionContext(ContextMenuActionContext:UIKit Context Menu Action Context Object Reference=@n27) -> (Output_Get:UIKit Context Menu Action Context Object Reference)
   > n28 = Branch(Condition=IsUsingTouch(PlayerController=GetOwningPlayer()))
   Event Construct()
   > Parent: Construct()
   n29 = Create Event()
   n30 = Get Anchor(self=Get ContextMenu())
   > Bind Event to On Menu Open Changed(self:Menu Anchor Object Reference=@n30, Delegate=@n29) {is=title K2Node_AddDelegate}
   > AdjustContextMenuTouch!(
    execute
    self
    IsTouchInputType:bool=IsUsingInputType!(InputType:ECommonInputType Enum=Touch, PlayerController:Player Controller Object Reference=GetOwningPlayer()) -> (ReturnValue:bool)
   ) -> (then) {is=title K2Node_CallFunction}
   > Set HideOutlineWhenCurrentSlotNotEquipped!(
    execute
    HideOutlineWhenCurrentSlotNotEquipped:bool=true
    self:WBP UIKit Equipped State Object Reference=Get EquippedState_Outline() -> (EquippedState_Outline:WBP UIKit Equipped State Object Reference)
   ) -> (then, Output_Get:bool)
   > n31 = Cast To FortTileView!(execute, Object=GetOwningListView!(UserListEntry:User List Entry Interface) -> (ReturnValue:List View Base Object Reference)) -> (then, CastFailed, AsFort Tile View:Fort Tile View Object Reference) {is=title K2Node_DynamicCast}
   > Set OwningFortTileView(OwningFortTileView:Fort Tile View Object Reference=@n31) -> (Output_Get:Fort Tile View Object Reference)
   Event Destruct()
   > Parent: Destruct()
   > Unbind Event from On Menu Open Changed(self:Menu Anchor Object Reference=@n30, Delegate=@n29) {is=title K2Node_RemoveDelegate}
   n32 = Event OnRemovedFromFocusPath!() -> (OutputDelegate, then, InFocusEvent:Focus Event Structure) {is=title K2Node_Event}
   > Parent: OnRemovedFromFocusPath!(execute, InFocusEvent:Focus Event Structure=@n32.InFocusEvent) -> (then) {is=title K2Node_CallParentFunction}
   n33 = GetCurrentInputType(self=Get CommonInputSubsystem())
   > n34 = Switch on ECommonInputType(Selection=@n33)
   n35 = GetSelected!(self:Common Button Base Object Reference) -> (ReturnValue:bool)
   n36 = AND Boolean!(A=@n35, B=Can Display Context Menu()) -> (ReturnValue)
   Branch(execute=@n34.Gamepad, Condition=@n36)
   > SetVisibility(self=Get ContextMenu(), InVisibility=Hidden)
   Event On Deselected()
   > Branch(Condition=Get IsButtonFocused() -> (IsButtonFocused:bool))
   > Parent: On Deselected() {is=title K2Node_CallParentFunction}
   > SetVisibility(self=Get ContextMenu(), InVisibility=Hidden)
   > Branch(Condition=IsUsingTouch(PlayerController=GetOwningPlayer()))
   > SetVisibility(self=Get UIKit_Transference() -> (UIKit_Transference:WBP UIKit Transference Object Reference), InVisibility=SelfHitTestInvisible)
   Event On Clicked()
   > Branch(Condition=NOT Boolean(A=Get IsEmoteWheel() -> (IsEmoteWheel:bool)))
   > Parent: On Clicked() {is=title K2Node_CallParentFunction}
   n37 = GetOwningPlayer()
   n38 = GetSelected()
   > n39 = Branch(Condition=AND Boolean(A=NOT Boolean(A=@n38), B=IsUsingTouch(PlayerController=@n37)))
   > PlayAnimationForward(InAnimation=Get Anim_Hovered() -> (Anim_Hovered:Widget Animation Object Reference))
   > n40 = SetFocusState!(execute, self:WBP UIKit Item Card Base Object Reference, IsFocused:bool=true) -> (then)
   SelectItem(execute=[@n52.then, @n51.else, @n40.then])
   > n41 = Cast To ListView!(execute, Object=GetOwningListView()) -> (then, CastFailed, AsList View:List View Object Reference) {is=title K2Node_DynamicCast}
   > Set Selected Item!(
    execute
    self:List View Object Reference=@n41
    Item:Object Reference=GetListItemObject!(UserObjectListEntry:User Object List Entry Interface) -> (ReturnValue:Object Reference)
   ) -> (then) {is=title K2Node_CallFunction}
   n42 = Cast To AthenaCosmeticItemDefinition!(execute=[@n54.then_0, @n39.else], Object=Get ItemDefinition(self=Get FortItemVM())) -> (then, CastFailed, AsAthena Cosmetic Item Definition:Athena Cosmetic Item Definition Object Reference) {is=title K2Node_DynamicCast}
   > n43 = Branch(Condition=NOT Boolean(A=IsCustomizable!() -> (Customizable:bool)))
   n44 = Get CustomItemDetails!(self:Fort Locker Category Config Object Reference=GetItemCategory(self=Get FortItemVM())) -> (CustomItemDetails:User Widget Soft Class Reference)
   n45 = Branch(
    execute=[@n43.then, @n42.CastFailed]
    Condition=IsValidSoftClassReference!(SoftClassReference:"Object Soft Class Reference (by ref)"=@n44) -> (ReturnValue:bool)
   )
   > n46 = Async Load Class Asset!(execute:Exec, AssetClass:Object Soft Class Reference=@n44) -> (then:Exec, Completed:Exec, Class:Object Class Reference) {is=title K2Node_LoadAssetClass}
   SelectItem(execute=@n46.Completed)
   > Call Go to Locker Item Details!(
    execute:Exec
    self:WBP UIKit Item Card Locker Object Reference
    CustomItemDetails:User Widget Class Reference=Cast To UserWidget Class!(Class:Object Class Reference=@n46) -> (AsUser Widget:User Widget Class Reference, bSuccess:bool=~) {is=title K2Node_ClassDynamicCast}
   ) -> (then:Exec) {is=title K2Node_CallDelegate}
   n47 = Get FortItemVM()
   n48 = ShowConfirmationOnEquipIfRequired!(
    execute:Exec=@n45.else
    Item:Fort Item VM Object Reference=@n47
    LockerConfirmationModalClass:Locker Confirmation Modal Soft Class Reference=/Game/Athena/UI/Customization/LockerArchivingModal.LockerArchivingModal_C
   ) -> (then:Exec, OnOperationCompleted:Exec, Success:bool=~) {is=title K2Node_AsyncAction}
   Branch(execute=@n48.OnOperationCompleted, Condition=@n48)
   n49 = Get FortItemSelectionVM()
   > n50 = EquipItem(
    Item=@n47
    SlotIndex=Get SlotIndex(self=Get SelectedCategory(self=@n49)) -> (SlotIndex=0)
    Variants=Make Array()
    bApplyToAllSlots=false
    bAllowMocking=true
   ) -> (Success=~)
   n51 = Branch(execute=@n50.OnOperationCompleted, Condition=@n50)
   > n52 = AnalyticsItemInteracted!(execute, self, Context:String=Equipped) -> (then) {is=title K2Node_CallFunction}
   n53 = Branch(execute=@n43.else, Condition=NOT Boolean(A=IsStreamingDataReady!() -> (ReturnValue:bool)))
   > Call On Wait for Streaming Content!(execute:Exec, self:WBP UIKit Item Card Locker Object Reference) -> (then:Exec) {is=title K2Node_CallDelegate}
   Branch(
    execute=@n53.else
    Condition=NOT Boolean(A=IsCosmeticDenied!(self:Athena Cosmetic Item Definition Object Reference=@n42) -> (ReturnValue:bool))
   )
   > Set SelectedItem!(execute, SelectedItem:Fort Item VM Object Reference=Get FortItemVM(), self:Fort Item Selection VM Object Reference=Get FortItemSelectionVM()) -> (then, Output_Get:Fort Item VM Object Reference)
   > Call On Context Menu Edit Styles Clicked()
   Event On Double Clicked() {is=title K2Node_Event}
   > Branch(Condition=NOT Boolean(A=Get IsEmoteWheel()))
   > Parent: On Double Clicked() {is=title K2Node_CallParentFunction}
   > n54 = Sequence()
   Call On Item Double Clicked!(execute:Exec=@n54.then_1, self:WBP UIKit Item Card Locker Object Reference) -> (then:Exec) {is=title K2Node_CallDelegate}
   Event On Hovered()
   > n55 = Branch(Condition=IsUsingTouch(PlayerController=GetOwningPlayer()))
   Set IsItemCardLockerHovered(execute=@n55.else, IsItemCardLockerHovered:bool=true) -> (Output_Get:bool)
   > Set SelectedItem(SelectedItem=Get FortItemVM(), self=Get FortItemSelectionVM())
   n56 = "Equal (Enum)"(A=GetCurrentInputType(self=Get CommonInputSubsystem()), B=MouseAndKeyboard)
   > n57 = Branch(Condition=AND Boolean(A=@n56, B=Can Display Context Menu()))
   > Branch(Condition=NOT Boolean(A=Get AwaitingContextMenuClose() -> (AwaitingContextMenuClose:bool)))
   n58 = Get ContextMenu()
   > SetVisibility(self=[Get ContextMenuButton(self=@n58), @n58], InVisibility=Visible)
   > n59 = Toggle Quick Actions!(execute, self:WBP UIKit Context Menu Anchor Button Base Object Reference=@n58, Enabled:bool=true) -> (then)
   Parent: On Hovered(execute=[@n59.then, @n57.else]) {is=title K2Node_CallParentFunction}
   > PlayAnimationForward(InAnimation=Get Anim_Hovered())
   n60 = Get ContextMenu()
   > n61 = Branch(Condition=IsOpen(self=Get Anchor(self=@n60)))
   Event On Unhovered()
   > n62 = Branch(Condition=IsUsingTouch(PlayerController=GetOwningPlayer()))
   Set IsItemCardLockerHovered(execute=@n62.else, IsItemCardLockerHovered=false)
   > MarkAsSeen()
   > n63 = Branch(Condition=@n56)
   n64 = Get Anchor(self=Get ContextMenu())
   > n65 = Branch(Condition=NOT Boolean(A=IsOpen(self=@n64)))
   n66 = Get ContextMenu()
   > SetVisibility(self=Get ContextMenuButton(self=@n66), InVisibility=Hidden)
   > n67 = Toggle Quick Actions(self=@n66, Enabled=false)
   Parent: On Unhovered(execute=[@n67.then, @n63.else]) {is=title K2Node_CallParentFunction}
   > PlayAnimationReverse(InAnimation=Get Anim_Hovered())
   n68 = Get ContextMenu()
   > n69 = Branch(Condition=IsOpen(self=Get Anchor(self=@n68)))
   RemoveAllInteractions(execute=@n69.else, self=Get ContextMenuActionContext())
   Set AwaitingContextMenuClose(execute=@n65.else, AwaitingContextMenuClose:bool=true) -> (Output_Get:bool)
   Event On Selected()
   > Parent: On Selected() {is=title K2Node_CallParentFunction}
   > n70 = Sequence()
   > Call On Locker Item Card Selected!(
    execute:Exec
    self:WBP UIKit Item Card Locker Object Reference
    FortTileView:Fort Tile View Object Reference=Get OwningFortTileView() -> (OwningFortTileView:Fort Tile View Object Reference)
   ) -> (then:Exec) {is=title K2Node_CallDelegate}
   n71 = Get FortItemVM()
   n72 = Get bIsValid(self=@n71)
   > n73 = Branch(Condition=OR Boolean(A=Get bIsEquipped!(self:Fort Item VM Object Reference=@n71) -> (bIsEquipped:bool), B=NOT Boolean(A=@n72)))
   > n74 = Set SelectedItem(SelectedItem=Get FortItemVM(), self=Get FortItemSelectionVM())
   Branch(execute=[@n73.else, @n74.then], Condition=IsUsingTouch(PlayerController=GetOwningPlayer()))
   > SetVisibility(self=Get UIKit_Transference(), InVisibility=Hidden)
   > n75 = RemoveAllInteractions(self=Get ContextMenuActionContext())
   n76 = GetCurrentInputType(self=Get CommonInputSubsystem())
   Branch(
    execute=@n70.then_1
    Condition=AND Boolean(
     A=Can Display Context Menu()
     B="Not Equal (Enum)"!(A:ECommonInputType Enum=@n76, B:ECommonInputType Enum=MouseAndKeyboard) -> (ReturnValue:bool)
    )
   )
   > SetVisibility(self=Get ContextMenu(), InVisibility=Visible)
   n77 = Event On Input Method Changed!() -> (OutputDelegate, then, CurrentInputType:ECommonInputType Enum) {is=title K2Node_Event}
   > n78 = Sequence()
   > n79 = Branch(Condition=Get AwaitingContextMenuClose())
   > Set AwaitingContextMenuClose(AwaitingContextMenuClose=false)
   n80 = Get ContextMenu()
   > n81 = Branch(Condition=IsOpen(self=Get Anchor(self=@n80)))
   > CloseContextMenu()
   > n82 = RemoveAllInteractions(self=Get ContextMenuActionContext())
   SetVisibility(execute=[@n82.then, @n81.else, @n79.else], self=Get ContextMenu(), InVisibility=Hidden)
   > AdjustContextMenuTouch(IsTouchInputType="Equal (Enum)"(A=@n77.CurrentInputType, B=Touch))
   Is Valid(exec=@n78.then_1, InputObject=Get FortItemVM())
   > RemoveAllInteractions(self=Get ContextMenuActionContext())
   > n83 = Switch on ECommonInputType(Selection=@n77.CurrentInputType)
   > n84 = Branch(Condition=GetSelected())
   > n85 = SetFocusState(IsFocused=false)
   Branch(execute=[@n85.then, @n84.else], Condition=Get bIsEquipped(self=Get FortItemVM()))
   > Set SelectedItem(SelectedItem=Get FortItemVM(), self=Get FortItemSelectionVM())
   Branch(execute=@n83.Touch, Condition=GetSelected())
   > SetFocusState(IsFocused=true)
   n86 = Event OnAddedToFocusPath!() -> (OutputDelegate, then, InFocusEvent:Focus Event Structure) {is=title K2Node_Event}
   > Parent: OnAddedToFocusPath!(execute, InFocusEvent:Focus Event Structure=@n86.InFocusEvent) -> (then) {is=title K2Node_CallParentFunction}
   > n87 = Switch on ECommonInputType(Selection=@n33)
   n88 = SetFocusState(execute=@n87.Touch, IsFocused=@n35)
   Branch(execute=[@n88.then, @n87.Gamepad], Condition=@n36)
   > SetVisibility(self=Get ContextMenu(), InVisibility=Visible)
   Call Is Modal Open(execute=@n89.then)
   Reroute Node(InputPin:Wildcard) -> (OutputPin:Wildcard)
   n89 = Unbind Event from On Widget Deactivated(
    execute=@n17.then
    self:Common Activatable Widget Object Reference=Get RenameModal() -> (RenameModal:Common Activatable Widget Object Reference)
    Delegate=@n17
   ) {is=title K2Node_RemoveDelegate}
  ]}
 GoToLockerItemDetails {nodes [GoToLockerItemDetails!() -> (then, CustomItemDetails:User Widget Class Reference) {is=title K2Node_FunctionEntry}]}
 GoToQuestsScreen {nodes [
   GoToQuestsScreen!() -> (then, InteractionNameID:Name) {is=title K2Node_FunctionEntry}
   n1 = Get SelectedItem(self=Get FortItemSelectionVM())
   > NavigateToQuestScreen!(
    execute
    Context:Widget Object Reference=Self-Reference()
    StartingCategoryTag:"Gameplay Tag Structure (by ref)"=GetConnectedQuestCategoryTag(ItemDefinition=Get ItemDefinition(self=@n1))
   ) -> (then)
  ]}
 HandleContextMenuOpenChanged {nodes [
   n1 = HandleContextMenuOpenChanged!() -> (then, IsOpen:bool=~) {is=title K2Node_FunctionEntry}
   > n2 = Branch(Condition=@n1)
   > Call On Context Menu Opened!(execute:Exec, self:WBP UIKit Item Card Locker Object Reference, ContextMenuWidgetParentEntry:Widget Object Reference=Self-Reference()) -> (then:Exec) {is=title K2Node_CallDelegate}
   Set AwaitingContextMenuClose(execute=@n2.else, AwaitingContextMenuClose=false)
   > n3 = Switch on ECommonInputType(Selection=GetCurrentInputType(self=Get CommonInputSubsystem()))
   n4 = SetVisibility(execute=@n3.Gamepad, self=Get ContextMenu(), InVisibility=Hidden)
   Call On Context Menu Closed!(execute:Exec=[@n4.then, @n3.Touch], self:WBP UIKit Item Card Locker Object Reference) -> (then:Exec) {is=title K2Node_CallDelegate}
   n5 = Branch(execute=@n3.MouseAndKeyboard, Condition=IsHovered())
   Call On Context Menu Closed(execute=[@n7.then, @n5.then])
   SetFocusState(execute=@n5.else, IsFocused=false)
   n6 = Get ContextMenu()
   > SetVisibility(self=[@n6, Get ContextMenuButton(self=@n6)], InVisibility=Hidden)
   > PlayAnimationReverse(InAnimation=Get Anim_Hovered())
   > n7 = RemoveAllInteractions(self=Get ContextMenuActionContext())
  ]}
 Is Item Equipped {nodes [
   n1 = Is Item Equipped!() -> (then, Category Item:Fort Item VM Object Reference, Selected Item:Fort Item VM Object Reference) {is=title K2Node_FunctionEntry}
   > Return Node!(execute, NewParam:bool="Equal (Object)"(A=@n1, B=@n1.Selected Item))
  ]}
 IsCustomizable {nodes [
   IsCustomizable() {is=title K2Node_FunctionEntry}
   > n1 = Get!(execute:Exec) -> (then:Exec, else:Exec, FortItemSelectionVM:Fort Item Selection VM Object Reference)
   > n2 = Get!(execute, self:Fort Item Selection VM Object Reference=@n1) -> (then, else, SelectedItem:Fort Item VM Object Reference)
   > Return Node!(execute, Customizable:bool=RequiresCustomization!(self:Fort Item VM Object Reference=@n2) -> (ReturnValue:bool))
   Return Node(execute=[@n2.else, @n1.else], Customizable=false)
  ]}
 IsCustomizableEnabled {nodes [
   n1 = IsCustomizableEnabled!() -> (then, Is Customizable:bool=~) {is=title K2Node_FunctionEntry}
   > SetIsItemCustomizable!(execute, self:WBP UIKit Item Card Base Object Reference, IsCustomizable:bool=@n1) -> (then)
   > Branch(Condition=IsInZone!(self:Fort Global UIContext Object Reference=Get FortGlobalUIContext() -> (ReturnValue:Fort Global UIContext Object Reference) {is=title K2Node_GetSubsystem}) -> (ReturnValue:bool))
   > SetRenderOpacity(
    self=[Get LazyImage_Item() -> (LazyImage_Item:Fort Lazy Image Object Reference), Get Background() -> (Background:Image Object Reference)]
    InOpacity=Select(Option 0:float=1, Option 1:float=0.5, Index=@n1) -> (ReturnValue:float)
   )
  ]}
 IsModalOpen {nodes [IsModalOpen!() -> (then, IsOpen:bool=~) {is=title K2Node_FunctionEntry}]}
 IsStreamingDataReady {nodes [
   IsStreamingDataReady() {is=title K2Node_FunctionEntry}
   n1 = Get FortDownloadSystemVM() -> (FortDownloadSystemVM:Fort Download System VM Object Reference)
   > Return Node!(
    execute
    ReturnValue:bool=Select(
     Option 0:bool=true
     Option 1:bool=GetIsBulkCosmeticInstalled!(self:Fort Download System VM Object Reference=@n1) -> (ReturnValue:bool)
     Index=GetIsBulkCosmeticEnabled!(self:Fort Download System VM Object Reference=@n1) -> (ReturnValue:bool)
    ) -> (ReturnValue:bool)
   )
  ]}
 MarkAsSeen {nodes [
   MarkAsSeen() {is=title K2Node_FunctionEntry}
   n1 = Get SelectedCategory(self=Get FortItemSelectionVM())
   > Is Valid(InputObject=@n1)
   > MarkItemAsSeen!(execute, self:Fort Locker Category VM Object Reference=Get Category(self=@n1), Item:Fort Item VM Object Reference=Get FortItemVM()) -> (then)
  ]}
 OnContextMenuClosed {nodes [OnContextMenuClosed() {is=title K2Node_FunctionEntry}]}
 OnContextMenuEditStylesClicked {nodes [OnContextMenuEditStylesClicked() {is=title K2Node_FunctionEntry}]}
 OnContextMenuOpened {nodes [OnContextMenuOpened!() -> (then, ContextMenuWidgetParentEntry:Widget Object Reference) {is=title K2Node_FunctionEntry}]}
 OnItemCardHovered {nodes [OnItemCardHovered() {is=title K2Node_FunctionEntry}]}
 OnItemDoubleClicked {nodes [OnItemDoubleClicked() {is=title K2Node_FunctionEntry}]}
 OnLockerItemCardSelected {nodes [OnLockerItemCardSelected!() -> (then, FortTileView:Fort Tile View Object Reference) {is=title K2Node_FunctionEntry}]}
 OnMouseButtonDown {nodes [
   n1 = OnMouseButtonDown!() -> (then, MyGeometry:Geometry Structure, MouseEvent:"Pointer Event Structure (by ref)") {is=title K2Node_FunctionEntry}
   > n2 = Branch(Condition=CanReactToRightClickDown!(Input:"Pointer Event Structure (by ref)"=@n1.MouseEvent) -> (ReturnValue:bool) {is=title K2Node_CallFunction})
   > Set Timer for Next Tick by Event!(execute, Delegate:Delegate=Create Event()) -> (then, ReturnValue:Timer Handle Structure) {is=title K2Node_CallFunction}
   > Return Node(ReturnValue:Event Reply Structure=Handled())
   Return Node(execute=@n2.else, ReturnValue=Handled())
  ]}
 OnWaitForStreamingContent {nodes [OnWaitForStreamingContent() {is=title K2Node_FunctionEntry}]}
 SelectCard {nodes [
   SelectCard() {is=title K2Node_FunctionEntry}
   > PlayAnimationForward(InAnimation=Get Anim_Hovered())
   > SetFocusState(IsFocused=true)
   > SelectItem()
   > n1 = Cast To ListView(Object=GetOwningListView())
   > Set Selected Item(self=@n1, Item=GetListItemObject())
   > Call On Locker Item Card Selected(FortTileView=Get OwningFortTileView())
  ]}
 SelectItem {nodes [
   SelectItem() {is=title K2Node_FunctionEntry}
   > Set SelectedItem(SelectedItem=Get FortItemVM(), self=Get FortItemSelectionVM())
  ]}
 Set Is New {nodes [
   n1 = Set Is New!() -> (then, IsNew:bool=~) {is=title K2Node_FunctionEntry}
   > SetItemIsNew!(execute, self:WBP UIKit Item Card Base Object Reference, Is New:bool=NOT Boolean(A=@n1)) -> (then)
  ]}
 SetCachedIndicatorConfig {nodes [
   SetCachedIndicatorConfig!() -> (
    then
    CachedIndicatorTypeToFind:E_UIKit_ItemCardStatusIndicator Enum=NewEnumerator5
    NewIndicatorConfig:E_UI_StatusIndicator_Configuration Enum=NewEnumerator2
   ) {is=title K2Node_FunctionEntry}
   > n1 = FindCachedStatusIndicator!(
    execute
    self:WBP UIKit Item Card Base Object Reference
    IndicatorType:E_UIKit_ItemCardStatusIndicator Enum=Get CachedIndicatorTypeToFind() -> (CachedIndicatorTypeToFind:E_UIKit_ItemCardStatusIndicator Enum=NewEnumerator10)
   ) -> (then, Status Indicator:WBP UIKit Status Indicator Base Object Reference, Valid:bool)
   > n2 = Branch(Condition=@n1.Valid)
   > SetIndicatorConfiguration!(
    execute
    self:WBP UIKit Status Indicator Base Object Reference=@n1
    InCount:int=0
    IndicatorConfig:E_UI_StatusIndicator_Configuration Enum=Get NewIndicatorConfig() -> (NewIndicatorConfig:E_UI_StatusIndicator_Configuration Enum=NewEnumerator0)
   ) -> (then)
   > Return Node!(execute, Success:bool=true)
   Return Node(execute=@n2.else, Success=false)
  ]}
 SetContextMenuActionContext {nodes [
   SetContextMenuActionContext() {is=title K2Node_FunctionEntry}
   > RemoveAllInteractions(self=Get ContextMenuActionContext())
   n1 = Get FortItemVM()
   > AddInteraction!(
    execute
    self:UIKit Context Menu Action Context Object Reference=Get ContextMenuActionContext()
    InActionInfo_DisplayName:Text=SelectText(
     A="LOCTEXT(\"Unarchive Item\")"
     B="LOCTEXT(\"Archive Item\")"
     bPickA=Get bIsArchived!(self:Fort Item VM Object Reference=@n1) -> (bIsArchived:bool)
    )
    InActionInfo_IconTexture:Texture 2D Soft Object Reference
    InActionInfo_IconMaterial:Material Instance Soft Object Reference
    InActionInfo_SortPriority:int=1200
    InActionInfo_bIsAvailable:bool=Get bCanArchive!(self:Fort Item VM Object Reference=@n1) -> (bCanArchive:bool)
    InActionInfo_bIsEnabled:bool=True
    InActionInfo_bIsSelected:bool=False
    InActionInfo_bAddDivider:bool=False
    InActionInfo_bPositionDividerAfter:bool=False
    InActionInfo_bCloseMenuOnSelect:bool=False
    InActionInfo_InteractionNameID:Name=None
    InActionInfo_Interaction:Data Table Row Handle Structure=Get ArchiveInput() -> (ArchiveInput:Data Table Row Handle Structure)
    InActionInfo_PressedDelegate:Delegate="(null).None"
    InActionInfo_ReleasedDelegate:Delegate="(null).None"
    InInteractionDelegate:"Delegate (by ref)"=Create Event()
   ) -> (then)
   n2 = Get FortItemVM()
   > AddInteraction(
    self=Get ContextMenuActionContext()
    InActionInfo_DisplayName=SelectText(A="LOCTEXT(\"Unfavorite Item\")", B="LOCTEXT(\"Favorite Item\")", bPickA=Get bIsFavorite(self=@n2))
    InActionInfo_SortPriority=1100
    InActionInfo_bIsAvailable=Get bCanFavorite!(self:Fort Item VM Object Reference=@n2) -> (bCanFavorite:bool)
    InActionInfo_bIsEnabled=True
    InActionInfo_bIsSelected=False
    InActionInfo_bAddDivider=False
    InActionInfo_bPositionDividerAfter=False
    InActionInfo_bCloseMenuOnSelect=False
    InActionInfo_InteractionNameID=None
    InActionInfo_Interaction=Get FavoriteInput() -> (FavoriteInput:Data Table Row Handle Structure)
    InActionInfo_PressedDelegate="(null).None"
    InActionInfo_ReleasedDelegate="(null).None"
    InInteractionDelegate=Create Event()
   )
   n3 = Get FortItemVM()
   > AddInteraction(
    self=Get ContextMenuActionContext()
    InActionInfo_DisplayName="LOCTEXT(\"Apply to All\")"
    InActionInfo_SortPriority=900
    InActionInfo_bIsAvailable=Get bCanApplyToAll!(self:Fort Locker Category Config Object Reference=GetItemCategory(self=@n3)) -> (bCanApplyToAll:bool)
    InActionInfo_bIsEnabled=True
    InActionInfo_bIsSelected=False
    InActionInfo_bAddDivider=False
    InActionInfo_bPositionDividerAfter=False
    InActionInfo_bCloseMenuOnSelect=False
    InActionInfo_InteractionNameID=None
    InActionInfo_Interaction=Get ApplyToAllInput() -> (ApplyToAllInput:Data Table Row Handle Structure)
    InActionInfo_PressedDelegate="(null).None"
    InActionInfo_ReleasedDelegate="(null).None"
    InInteractionDelegate=Create Event()
   )
   n4 = Get FortGlobalUIContext()
   > n5 = Branch(Condition=AND Boolean(A=IsInZone(self=@n4), B=IsCustomizable() {is=title K2Node_CallFunction}))
   n6 = IsInZone(self=Get FortGlobalUIContext())
   n7 = IsCustomizable()
   n8 = GetItemType(ItemDefinition=Get ItemDefinition(self=Get FortItemVM())) -> (ReturnValue=WorldItem)
   n9 = Branch(
    execute=[@n21.then, @n5.then]
    Condition=AND Boolean!(A="Equal (Enum)"(A:EFortItemType Enum=@n8, B:EFortItemType Enum=CosmeticMimosa), B=NOT Boolean(A=@n7), C:bool=NOT Boolean(A=@n6)) -> (ReturnValue)
   )
   > n10 = IsNameable!(
    execute:Exec
    self:Object Reference=Get ItemDetails!(self:Fort Item VM Object Reference=Get FortItemVM()) -> (ItemDetails:Fort Item Details VM Object Reference)
   ) -> (then:Exec, ReturnValue:bool) {is=title K2Node_Message}
   > n11 = AddInteraction(
    self=Get ContextMenuActionContext()
    InActionInfo_DisplayName="LOCTEXT(\"Rename\")"
    InActionInfo_SortPriority=980
    InActionInfo_bIsAvailable=@n10
    InActionInfo_bIsEnabled=True
    InActionInfo_bIsSelected=False
    InActionInfo_bAddDivider=False
    InActionInfo_bPositionDividerAfter=False
    InActionInfo_bCloseMenuOnSelect=False
    InActionInfo_InteractionNameID=None
    InActionInfo_Interaction=Get RenameModalInput() -> (RenameModalInput:Data Table Row Handle Structure)
    InActionInfo_PressedDelegate="(null).None"
    InActionInfo_ReleasedDelegate="(null).None"
    InInteractionDelegate=Create Event()
   )
   n12 = IsInZone(self=Get FortGlobalUIContext())
   n13 = IsInSubGame!(SubGame:ESubGame Enum=Campaign) -> (ReturnValue:bool)
   n14 = Branch(execute=[@n11.then, @n9.else], Condition=AND Boolean!(A=NOT Boolean(A=@n12), B=NOT Boolean(A=@n13)) -> (ReturnValue))
   > n15 = AddInteraction(
    self=Get ContextMenuActionContext()
    InActionInfo_DisplayName="LOCTEXT(\"Go to Quests\")"
    InActionInfo_SortPriority=1250
    InActionInfo_bIsAvailable=Get IsQuestAvailable() -> (IsQuestAvailable:bool)
    InActionInfo_bIsEnabled=True
    InActionInfo_bIsSelected=False
    InActionInfo_bAddDivider=False
    InActionInfo_bPositionDividerAfter=False
    InActionInfo_bCloseMenuOnSelect=False
    InActionInfo_InteractionNameID=None
    InActionInfo_Interaction=Get GoToQuestsInput() -> (GoToQuestsInput:Data Table Row Handle Structure)
    InActionInfo_PressedDelegate="(null).None"
    InActionInfo_ReleasedDelegate="(null).None"
    InInteractionDelegate=Create Event()
   )
   n16 = GetItemTemplateId(self=Get FortItemVM())
   SetParentIDForContextMenuTelemetry!(
    execute=[@n15.then, @n14.else]
    self:WBP UIKit Context Menu Anchor Button Base Object Reference=Get ContextMenu()
    ParentID:String=Get Source String from Text!(Text:Text=Format Text!(Format="LOCTEXT(\"LockerItem {1}\")", 1=@n16) -> (Result)) -> (ReturnValue:String) {is=title K2Node_CallFunction}
    IsParentIdEGPI:bool=false
   ) -> (then)
   > SetContext!(
    execute
    self:WBP UIKit Context Menu Anchor Button Base Object Reference=Get ContextMenu()
    InteractionContext_Data:UIKit Context Menu Action Context Object Reference=Get ContextMenuActionContext()
   ) -> (then)
   n17 = IsCustomizable()
   n18 = Get FortItemVM()
   n19 = MakeLiteralText(Value="LOCTEXT(\"Edit Styles\")")
   n20 = Get bIsShuffleTile(self=@n18)
   n21 = AddInteraction(
    execute=@n5.else
    self=Get ContextMenuActionContext()
    InActionInfo_DisplayName=Select(Option 0:Text=SelectText(A="LOCTEXT(\"Edit Random Settings\")", B=@n19, bPickA=@n20), Option 1:Text="LOCTEXT(\"Customize\")", Index=@n17) -> (ReturnValue:Text)
    InActionInfo_SortPriority=1000
    InActionInfo_bIsAvailable=Select(Option 0:bool=Get bHasVariants!(self:Fort Item VM Object Reference=@n18) -> (bHasVariants:bool), Option 1:bool=true, Index=@n17) -> (ReturnValue:bool)
    InActionInfo_bIsEnabled=Select(Option 0=true, Option 1=IsStreamingDataReady() {is=title K2Node_CallFunction}, Index=IsCustomizable())
    InActionInfo_bIsSelected=False
    InActionInfo_bAddDivider=False
    InActionInfo_bPositionDividerAfter=False
    InActionInfo_bCloseMenuOnSelect=False
    InActionInfo_InteractionNameID=None
    InActionInfo_Interaction=Get StylesInput() -> (StylesInput:Data Table Row Handle Structure)
    InActionInfo_PressedDelegate="(null).None"
    InActionInfo_ReleasedDelegate="(null).None"
    InInteractionDelegate=Create Event()
   )
  ]}
 SetContextMenuPrimaryPreviewActions {nodes [
   n1 = SetContextMenuPrimaryPreviewActions!() -> (then, PreviewActionName:Text, PreviewActionInInteractionDelegate:"Delegate (by ref)") {is=title K2Node_FunctionEntry}
   n2 = String To Name(InString="To String (Text)"(InText=@n1))
   n3 = Get ContextMenuActionContext()
   > RemoveInteraction!(execute, self:UIKit Context Menu Action Context Object Reference=@n3, InteractionNameID:Name=@n2) -> (then)
   > AddInteraction(
    self=@n3
    InActionInfo_DisplayName=@n1
    InActionInfo_SortPriority=1000
    InActionInfo_bIsAvailable=True
    InActionInfo_bIsEnabled=True
    InActionInfo_bIsSelected=False
    InActionInfo_bAddDivider=False
    InActionInfo_bPositionDividerAfter=False
    InActionInfo_bCloseMenuOnSelect=False
    InActionInfo_InteractionNameID=@n2
    InActionInfo_Interaction=(DataTable=None,RowName="")
    InActionInfo_PressedDelegate="(null).None"
    InActionInfo_ReleasedDelegate="(null).None"
    InInteractionDelegate=@n1.PreviewActionInInteractionDelegate
   )
   > SetContext(self=Get ContextMenu(), InteractionContext_Data=@n3)
  ]}
 SetContextMenuSecondaryPreviewActions {nodes [
   n1 = SetContextMenuSecondaryPreviewActions!() -> (then, PreviewActionName:Text, PreviewActionInInteractionDelegate:"Delegate (by ref)") {is=title K2Node_FunctionEntry}
   n2 = String To Name(InString="To String (Text)"(InText=@n1))
   n3 = Get ContextMenuActionContext()
   > RemoveInteraction(self=@n3, InteractionNameID=@n2)
   > AddInteraction(
    self=@n3
    InActionInfo_DisplayName=@n1
    InActionInfo_SortPriority=1000
    InActionInfo_bIsAvailable=True
    InActionInfo_bIsEnabled=True
    InActionInfo_bIsSelected=False
    InActionInfo_bAddDivider=False
    InActionInfo_bPositionDividerAfter=False
    InActionInfo_bCloseMenuOnSelect=False
    InActionInfo_InteractionNameID=@n2
    InActionInfo_Interaction=(DataTable=None,RowName="")
    InActionInfo_PressedDelegate="(null).None"
    InActionInfo_ReleasedDelegate="(null).None"
    InInteractionDelegate=@n1.PreviewActionInInteractionDelegate
   )
   > SetContext(self=Get ContextMenu(), InteractionContext_Data=@n3)
  ]}
 SetFIVMItemAndSlot {nodes [
   SetFIVMItemAndSlot() {is=title K2Node_FunctionEntry}
   n1 = Get FortItemSelectionVM()
   > Set SelectedItem(SelectedItem=Get FortItemVM(), self=@n1)
   > Return Node!(
    execute
    Selected Item:Fort Item VM Object Reference=Get SelectedItem(self=@n1)
    Slot Index:int=Get SlotIndex(self=Get SelectedCategory(self=@n1)) -> (SlotIndex=0)
   )
  ]}
 SetItemIsRestricted {nodes [
   n1 = SetItemIsRestricted!() -> (then, IsCompatible:bool) {is=title K2Node_FunctionEntry}
   > Parent: SetItemIsRestricted!(execute, IsCompatible:bool=@n1) -> (then) {is=title K2Node_CallParentFunction}
   > SetOpacity!(execute, self:Image Object Reference=Get Background(None=Self-Reference()), InOpacity:float=SelectFloat(A=0.25, B=1, bPickA=@n1)) -> (then)
  ]}
 ShowRewardPathingGlow {nodes [
   n1 = ShowRewardPathingGlow!() -> (then, Show:bool=~) {is=title K2Node_FunctionEntry}
   > n2 = Branch(Condition=@n1)
   > SetVisibility(self=Get RewardPathingGlow() -> (RewardPathingGlow:Image Object Reference), InVisibility=SelfHitTestInvisible)
   n3 = Get RewardPathingGlow()
   Branch(
    execute=@n2.else
    Condition="Equal (Enum)"(
     A:ESlateVisibility Enum=Get Visibility!(self:Widget Object Reference=@n3) -> (Visibility:ESlateVisibility Enum=Visible)
     B:ESlateVisibility Enum=SelfHitTestInvisible
    )
   )
   > PlayAnimation(InAnimation=Get FadeOutRewardPathing() -> (FadeOutRewardPathing:Widget Animation Object Reference))
  ]}
 Update Equipped Indexes {nodes [
   n1 = Update Equipped Indexes!() -> (then, Equipped Indexes:Array of Integers) {is=title K2Node_FunctionEntry}
   > n2 = Branch(Condition=Is Empty!(TargetArray:Array of Integers=@n1) -> (ReturnValue:bool))
   > SetItemIsEquippedSecondary!(
    execute
    self:WBP UIKit Item Card Base Object Reference
    Is Equipped:bool=false
    ShowEquippedSlotIndex:bool=false
    EquippedSlotIndex:int=0
    SlotIcon:Soft Object Reference
   ) -> (then)
   n3 = GetItemCategory(self=Get FortItemVM())
   n4 = "Get (a copy)"(Array:Array of Integers=Get Equipped Indexes() -> (Equipped Indexes:Array of Integers)) -> (Output:int)
   SetItemIsEquippedSecondary(
    execute=@n2.else
    Is Equipped=true
    ShowEquippedSlotIndex=Get bSecondaryEquipShowNumbers!(self:Fort Locker Category Config Object Reference=@n3) -> (bSecondaryEquipShowNumbers:bool)
    EquippedSlotIndex=int + int(A=@n4)
    SlotIcon=Break Fort Locker Category Slot Info!(FortLockerCategorySlotInfo:"Fort Locker Category Slot Info Structure (by ref)"=GetSlotInfo!(self:Fort Locker Category Config Object Reference=@n3, SlotIndex:int=@n4) -> (ReturnValue:"Fort Locker Category Slot Info Structure (by ref)")) -> (
     CategoryName:Text
     CategoryNamePlural:Text
     CategoryNameWithPluralForms:Text
     CategoryDescription:Text
     DefaultEmptyTexture:Soft Object Reference
     OverrideShuffleTexture:Texture 2D Soft Object Reference
     CategoryIcon:Material Instance Object Reference
     CategorySoundOverrides:Map of Gameplay Tag Structures to Sound Base Soft Object References
     bShowInCategoryTab:bool=~
     PreviewWindowScale:float=~
     PreviewWindowOffset:Vector 3f
     bUsePluralCategoryInItemList:bool=~
    ) {is=title K2Node_BreakStruct}.DefaultEmptyTexture
   )
  ]}
 VM_UpdateIsQuestAvailable {nodes [
   n1 = VM_UpdateIsQuestAvailable!() -> (then, InGameplayTags:Gameplay Tag Structure) {is=title K2Node_FunctionEntry}
   > n2 = Cast To FortPlayerController(Object=GetOwningPlayer())
   > n3 = PlayerHasQuestsForCategory!(execute, FortPC:Fort Player Controller Object Reference=@n2, QuestCategoryTag:"Gameplay Tag Structure (by ref)"=@n1) -> (then, ReturnValue:bool)
   > Set IsQuestAvailable(IsQuestAvailable:bool=AND Boolean(A=@n3, B=IsGameplayTagValid!(GameplayTag:Gameplay Tag Structure=@n1) -> (ReturnValue:bool))) -> (Output_Get:bool)
  ]}
]
variables [
 AwaitingContextMenuClose {type=boolean}
 ContextMenuActionContext {type=UIKitContextMenuActionContext, default=null}
 Favorite Sound {type=SoundBase, default=/Game/Sounds/Fort_UI_Sounds/FrontEnd/MultiProduct/Common/Buttons/MSS_UI_MP_Button_Click_Favorite}
 GoToLockerItemDetails {}
 IsEmoteWheel {type=boolean}
 IsItemCardLockerHovered {type=boolean}
 IsModalOpen {}
 IsQuestAvailable {type=boolean}
 OnContextMenuClosed {}
 OnContextMenuEditStylesClicked {}
 OnContextMenuOpened {}
 OnItemCardHovered {}
 OnItemDoubleClicked {}
 OnLockerItemCardSelected {}
 OnWaitForStreamingContent {}
 OwningFortTileView {type=FortTileView, default=null}
 RenameModal {type=CommonActivatableWidget, default=null}
 Right Clicked Sound {type=SoundBase, default=/Game/Sounds/Fort_UI_Sounds/FrontEnd/DefaultSounds/PressedButtons/MSS_UI_Button_Pressed_Default}
 Unfavorite Sound {type=SoundBase, default=/Game/Sounds/Fort_UI_Sounds/FrontEnd/MultiProduct/Common/Buttons/MSS_UI_MP_Button_Click_Unfavorite}
]
widgets {
 info {
  inheritedWidgetCount=0
  namedSlotCount=2
  parentClass=/FortUILibrary/Systems/Items/WBP_UIKit_ItemCard_Base.WBP_UIKit_ItemCard_Base_C
  rootWidgetClass=null
  widgetCount=4
 }
 widgets [
  ContextMenu_Attachment {
   widgetClassPath=&FortVisualAttachment
   bIsVariable=true
   widget=ContextMenu_Attachment
   properties {bOverride_HeightOverride=true, bOverride_WidthOverride=true, heightOverride=40, visibility=Visible, widthOverride=40, contentAnchor=<1 1>}
   slot=~
  }
  ContextMenu {
   widgetClassPath=/FortUILibrary/UIKit/ContextMenu/WBP_UIKit_ContextMenuAnchorButton
   bIsVariable=true
   parent=ContextMenu_Attachment
   widget=ContextMenu
   properties {
    button Background Material=/FortUILibrary/UIKit/Buttons/MaterialInstances/CTAButtons/MI_UIKit_BgQuiet
    buttonClass=/FortUILibrary/UIKit/Buttons/WBP_UIKit_Button_Contextual.WBP_UIKit_Button_Contextual_C
    buttonIcon=/FortUILibrary/Textures/Icons/T_UI_Icon_UIKit_ContextMenu
    buttonText=""
    button Padding=<ltrb 4 4 4 4>
    button Text Font {
     bForceMonospaced=false
     bMaterialIsStencil=false
     fontMaterial=/FortUILibrary/UIKit/Buttons/MaterialInstances/GenericButtons/MI_UIKit_GenericText_Quiet
     fontObject=&Burbank
     letterSpacing=180
     monospacedWidth=1
     size=26.25
     skewAmount=0
     typefaceFontName=BurbankRegular-Black
     outlineSettings {
      bApplyOutlineToDropShadows=false
      bMiteredCorners=false
      bSeparateFillAlpha=false
      outlineMaterial=null
      outlineSize=0
      outlineColor=<rgba 0 0 0 1>
     }
    }
   }
   slot {class=SizeBoxSlot, properties {horizontalAlignment=HAlign_Right, verticalAlignment=VAlign_Bottom, padding=<ltrb -8 -8 8 8>}}
  }
  Overlay_0 {widgetClassPath=&Overlay, widget=Overlay_0}
  RewardPathingGlow {
   widgetClassPath=&Image
   bIsVariable=true
   parent=Overlay_0
   widget=RewardPathingGlow
   properties {
    visibility=Hidden
    brush {
     bIsDynamicallyLoaded=false
     drawAs=Image
     imageType=FullColor
     mirroring=NoMirror
     resourceName=None
     resourceObject=/FortUILibrary/MaterialInstances/Effects/MI_UI_RewardState_Locker
     tiling=NoTile
     imageSize=<32 32>
     margin=<ltrb 0 0 0 0>
     outlineSettings {
      bUseBrushTransparency=false
      roundingType=HalfHeightRadius
      width=0
      color {colorUseRule=UseColor_Specified, specifiedColor=<rgba 0 0 0 0>}
      cornerRadii=<0 0 0 0>
     }
     tintColor {colorUseRule=UseColor_Specified, specifiedColor=<rgba 1 1 1 1>}
     uVRegion {bIsValid=false, max=<0 0>, min=<0 0>}
    }
   }
   slot {class=OverlaySlot, properties {horizontalAlignment=HAlign_Fill, verticalAlignment=VAlign_Fill}}
  }
 ]
}
